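# Looping mono MetaSound source that plays randomized cloud lightning one-shots for the Major Spire ambience.
# Two parallel layers, long and short waves, each wait a random delay, play a non-repeating pick with random pitch, then retrigger through deferred trigger variables.
# Each finished wave rerolls its layer's mixer gain and fires OnLightning.
# The mixed signal's absolute value is threshold-gated, scaled, clamped and smoothed into the LightningAmplitude output.
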MetaSoundSource /SpireSharedAssets/Audio/MetaSounds/MSS_MajorSpire_Cloud_Lightning
inputs [
 Array 1 {type=WaveAsset:Array}
 Array 2 {type=WaveAsset:Array}
 UE.Source.OnPlay {type=Trigger}
]
interfaces [UE.OutputFormat.Mono:1.0, UE.Source:1.1]
outputs [
 LightningAmplitude {type=Float}
 OnLightning {type=Trigger}
 "UE.OutputFormat.Mono.Audio:0" {type=Audio}
]
pages [Default {
  input_defaults {
   UE.Source.OnPlay=null
   Array 1 [
    /SpireSharedAssets/Audio/Waves/MajorSpire_Cloud_Lightning_01
    /SpireSharedAssets/Audio/Waves/MajorSpire_Cloud_Lightning_02
    /SpireSharedAssets/Audio/Waves/MajorSpire_Cloud_Lightning_03
    /SpireSharedAssets/Audio/Waves/MajorSpire_Cloud_Lightning_04
    /SpireSharedAssets/Audio/Waves/MajorSpire_Cloud_Lightning_05
    /SpireSharedAssets/Audio/Waves/MajorSpire_Cloud_Lightning_06
    /SpireSharedAssets/Audio/Waves/MajorSpire_Cloud_Lightning_07
    /SpireSharedAssets/Audio/Waves/MajorSpire_Cloud_Lightning_08
    /SpireSharedAssets/Audio/Waves/MajorSpire_Cloud_Lightning_09
    /SpireSharedAssets/Audio/Waves/MajorSpire_Cloud_Lightning_10
    /SpireSharedAssets/Audio/Waves/MajorSpire_Cloud_Lightning_11
    /SpireSharedAssets/Audio/Waves/MajorSpire_Cloud_Lightning_12
   ]
   Array 2 [
    /SpireSharedAssets/Audio/Waves/MajorSpire_Cloud_Lightning_Short_01
    /SpireSharedAssets/Audio/Waves/MajorSpire_Cloud_Lightning_Short_02
    /SpireSharedAssets/Audio/Waves/MajorSpire_Cloud_Lightning_Short_03
    /SpireSharedAssets/Audio/Waves/MajorSpire_Cloud_Lightning_Short_04
    /SpireSharedAssets/Audio/Waves/MajorSpire_Cloud_Lightning_Short_05
    /SpireSharedAssets/Audio/Waves/MajorSpire_Cloud_Lightning_Short_06
    /SpireSharedAssets/Audio/Waves/MajorSpire_Cloud_Lightning_Short_07
    /SpireSharedAssets/Audio/Waves/MajorSpire_Cloud_Lightning_Short_08
    /SpireSharedAssets/Audio/Waves/MajorSpire_Cloud_Lightning_Short_09
    /SpireSharedAssets/Audio/Waves/MajorSpire_Cloud_Lightning_Short_10
    /SpireSharedAssets/Audio/Waves/MajorSpire_Cloud_Lightning_Short_11
    /SpireSharedAssets/Audio/Waves/MajorSpire_Cloud_Lightning_Short_12
   ]
  }
  nodes [
   n1 = VariableDeferredAccessor.Trigger(Variable=InitVariable.Trigger().Variable)
   TriggerAnyTriggerAny2_2 = "TriggerAny.Trigger Any (2)"(In 0=^UE.Source.OnPlay, In 1=@n1.Value)
   n2 = UE.RandomTime(Next=@TriggerAnyTriggerAny2_2.Out) {also {defaults {Max=1.5, Min=0.5}}}
   TriggerDelay_1 = UE.Trigger Delay(In=@n2.On Next, Delay Time=@n2.Value)
   ArrayRandomGet_1 = Array.Random Get:WaveAsset:Array(In Array=^Array 1, Next=@TriggerDelay_1.Out) {also {defaults {No Repeats=3, Weights []}}}
   n3 = UE.RandomFloat(Next=@ArrayRandomGet_1.On Next) {also {defaults {Max=1.1, Min=0.9}}}
   WavePlayer_0 = UE.Wave Player:Mono(Wave Asset=@ArrayRandomGet_1.Value, Play=@n3.On Next, Pitch Shift=@n3.Value)
   n4 = UE.RandomFloat(Next=@WavePlayer_0.On Finished) {also {defaults {Min=0.3}}}
   n5 = VariableDeferredAccessor.Trigger(Variable=InitVariable.Trigger().Variable)
   TriggerAnyTriggerAny2_1 = "TriggerAny.Trigger Any (2)"(In 0=^UE.Source.OnPlay, In 1=@n5.Value)
   n6 = UE.RandomTime(Next=@TriggerAnyTriggerAny2_1.Out) {also {defaults {Max=1.5, Min=0.5}}}
   TriggerDelay_0 = UE.Trigger Delay(In=@n6.On Next, Delay Time=@n6.Value)
   ArrayRandomGet_0 = Array.Random Get:WaveAsset:Array(In Array=^Array 2, Next=@TriggerDelay_0.Out) {also {defaults {No Repeats=3}}}
   n7 = UE.RandomFloat(Next=@ArrayRandomGet_0.On Next) {also {defaults {Max=1.1, Min=0.9}}}
   WavePlayer_1 = UE.Wave Player:Mono(Wave Asset=@ArrayRandomGet_0.Value, Play=@n7.On Next, Pitch Shift=@n7.Value)
   n8 = UE.RandomFloat(Next=@WavePlayer_1.On Finished) {also {defaults {Min=0.3}}}
   AudioMixerAudioMixerMono2_0 = "AudioMixer.Audio Mixer (Mono, 2)"(Gain 0=@n4.Value, Gain 1=@n8.Value, In 0=@WavePlayer_0.Out Mono, In 1=@WavePlayer_1.Out Mono)
   n9 = AbsoluteValue.Abs:Float(Input=UE.ConversionAudioToFloat(In=@AudioMixerAudioMixerMono2_0.Out).Out)
   TriggerOnValueChange_0 = UE.Trigger On Value Change:Float(Value=@n9.Value)
   n10 = TriggerCompare.Clamp:Float(A=@n9.Value, Compare=@TriggerOnValueChange_0.Trigger) {also {defaults {B=0.1, Type=3}}}
   TriggerRouteTriggerRouteFloat2_0 = "TriggerRoute.Trigger Route (Float, 2):Float"(Value 0=@n9.Value, Set 1=@n10.False, Set 0=@n10.True)
   InterpTo_0 = UE.InterpTo:Audio(Target=Clamp.Clamp:Float(In=UE.Multiply:Float(PrimaryOperand=@TriggerRouteTriggerRouteFloat2_0.Value) {also {defaults {AdditionalOperands=10}}}.Out) {also {defaults {Max=1}}}.Value) {also {defaults {Interp Time=1}}}
   TriggerAnyTriggerAny2_0 = "TriggerAny.Trigger Any (2)"(In 0=@n4.On Next, In 1=@n8.On Next)
   VariableMutator.Trigger(Value=@n4.On Next, Variable=@n1.Variable)
   VariableMutator.Trigger(Value=@n8.On Next, Variable=@n5.Variable)
  ]
  links [
   {dst="@UE.OutputFormat.Mono.Audio:0", src=AudioMixerAudioMixerMono2_0.Out}
   {dst="@LightningAmplitude", src=InterpTo_0.Value}
   {dst="@OnLightning", src=TriggerAnyTriggerAny2_0.Out}
  ]
 }]
properties {
 assetImportData=&self:AssetImportData
 attenuationSettings=/SpireSharedAssets/Audio/Settings/ATT_MajorSpire_Cloud
 bLooping=true
 duration=10000
 maxDistance=28000
 sampleRate=48000
 soundClassObject=&Fort_Ambient_OneShots_Outdoors
 volume=0.3
 concurrencyOverrides {maxCount=2}
 concurrencySet [/SpireSharedAssets/Audio/Settings/Conc_MajorSpire_Ambient_Lightning]
}
